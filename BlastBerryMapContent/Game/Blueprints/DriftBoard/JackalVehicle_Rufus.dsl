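; Jackal vehicle (DriftBoard) Blueprint for Rufus that adds cosmetic water effects and sounds.
; On water entry it plays a splash and, above the lowest detail mode, starts looping FX, audio and a timer that tracks water height, velocity direction and speed-based volume.
; Exit deactivates the FX and fades the loop.
; Overlap begin sets the FX height offset by lake type, defaulting for any other actor, and prints debug strings.
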(event Vehicle|Water|EventOnEnteredWaterVolume
  (params (WaterSurfacePoint :type Vector%20%28by%20ref%29 :default "0, 0, 0"))
  (bind _returnvalue (CallFunction|GetWaterLevel))
  (bind (_returnvalue_x _returnvalue_y _returnvalue_z) (Math|Vector|BreakVector (Transformation|GetActorLocation)))
  (bind _watervfxoffset (Variables|Default|GetWaterVFXOffset))
  (bind _returnvalue_2 (+ _returnvalue _watervfxoffset))
  (bind _returnvalue_3 (- _returnvalue_2 _returnvalue_z))
  (bind _skeletalmesh (Variables|Vehicle|GetSkeletalMesh))
  ;[1 1 1] FX for water ineraction
  (Networking|CanExecuteCosmeticEvents
    (:True
      (if (> (Rendering|GetRenderingDetailMode) 0)
        (bind _returnvalue_4 (Utilities|Time|SetTimerbyEvent (delegate Custom|Check%20Water%20Loop) 0.25 true))
        (bind _inwaterfx (Variables|Default|GetInWaterFX)
          (:then
            (Components|DestroyComponent _inwaterfx))
          (:else
            _))
        (bind _returnvalue_5 (Effects|SpawnEmitterAttached :EmitterTemplate "/Game/WaterAndWind/Particles/WaterInteraction/P_Vehicle_In_Water_Athena.P_Vehicle_In_Water_Athena" :AttachToComponent _skeletalmesh :AttachPointName "None" :Location (Math|Vector|MakeVector 0.0 0.0 _returnvalue_3) :Scale "1.000000,1.000000,1.000000" :LocationType "KeepRelativeOffset" :bAutoDestroy true :PoolingMethod "None" :bAutoActivate true)) ; Looping water FX
        (Variables|Default|SetInWaterFX _returnvalue_5)
        (Effects|SpawnEmitterAttached :EmitterTemplate "/Game/WaterAndWind/Particles/WaterInteraction/P_WaterSplash_Athena_Vehicle.P_WaterSplash_Athena_Vehicle" :AttachToComponent _skeletalmesh :AttachPointName "None" :Location (Math|Vector|MakeVector 0.0 0.0 _returnvalue_3) :Scale "1.000000,1.000000,1.000000" :LocationType "KeepRelativeOffset" :bAutoDestroy true :PoolingMethod "None" :bAutoActivate true) ; Splash burst FX
        (bind _inwaterloop (Variables|Default|GetInWaterLoop)
          (:then
            (Audio|Components|Audio|Stop _inwaterloop))
          (:else
            _))
        (Audio|SpawnSoundAttached "/Game/Athena/Sounds/Vehicles/GolfCart/GolfCart_Movement_Impact_Water_Cue.GolfCart_Movement_Impact_Water_Cue" (Variables|Vehicle|GetSkeletalMesh))
        (bind _returnvalue_6 (Audio|SpawnSoundAttached "/Game/Athena/Sounds/Vehicles/GolfCart/GolfCart_Movement_InWater_Loop_Cue.GolfCart_Movement_InWater_Loop_Cue" (Variables|Vehicle|GetSkeletalMesh)))
        (Variables|Default|SetInWaterLoop _returnvalue_6)))))

(event Custom|Check%20Water%20Loop
  (bind (_returnvalue_x _returnvalue_y _returnvalue_z) (Math|Vector|BreakVector (Transformation|GetActorLocation)))
  ;[1 1 1] FX for water ineraction
  (bind _inwaterfx (Variables|Default|GetInWaterFX)
    (:then
      (Transformation|SetRelativeLocation _inwaterfx (Math|Vector|MakeVector 0.0 0.0 (- (+ (CallFunction|GetWaterLevel) (Variables|Default|GetWaterVFXOffset)) _returnvalue_z)) false false)
      (Effects|Components|ParticleSystem|SetVectorParameter (Variables|Default|GetInWaterFX) "Offset" (Math|Vector|Normalize (Math|Transform|InverseTransformDirection (Transformation|GetWorldTransform (Variables|Default|GetInWaterFX)) (Transformation|GetVelocity))))
      (bind _inwaterloop (Variables|Default|GetInWaterLoop)
        (:then
          (Audio|Components|Audio|SetVolumeMultiplier _inwaterloop (Math|Float|MapRangeClamped (Vehicle|GetForwardSpeedKmh) 0.0 2e+01 0.0 1.0) :!class /Script/Engine.AudioComponent))
        (:else)))
    (:else)))

(event Vehicle|Water|EventOnExitedWaterVolume
  (params (WaterSurfacePoint :type Vector%20%28by%20ref%29 :default "0, 0, 0"))
  ;[1 1 1] FX for water ineraction
  (Networking|CanExecuteCosmeticEvents
    (:True
      (Utilities|Time|ClearandInvalidateTimerbyHandle (stale Vehicle|Water|EventOnEnteredWaterVolume _returnvalue_4))
      (bind _inwaterfx (Variables|Default|GetInWaterFX)
        (:then
          (Components|Activation|SetActive _inwaterfx false false)
          (bind _inwaterloop (Variables|Default|GetInWaterLoop)
            (:then
              (Audio|Components|Audio|FadeOut _inwaterloop 3.0 0.0 "Linear" :!class /Script/Engine.AudioComponent))
            (:else)))
        (:else)))))

(event Collision|EventActorBeginOverlap
  (params (OtherActor :type Actor%20Object%20Reference))
  (bind _asbp_rufus_water (Utilities|Casting|CastToBP_Rufus_Water OtherActor))
  (Default|UpdateWaterVFXOffset true _asbp_rufus_water))

(event Collision|EventActorEndOverlap
  (params (OtherActor :type Actor%20Object%20Reference))
  (bind _asbp_rufus_water (Utilities|Casting|CastToBP_Rufus_Water OtherActor))
  (Default|UpdateWaterVFXOffset false _asbp_rufus_water))

(fn UpdateWaterVFXOffset
  (params (Entering :type Boolean) (Lake%20Reference :type BP%20Rufus%20Water%20Object%20Reference))
  (if Entering
    (Utilities|Casting|CastToBP_Rufus_Water_LeakyLake Lake%20Reference
      (:then
        (Development|PrintString "Vehicle is entering Leaky Lake.")
        (Variables|Default|SetWaterVFXOffset -6e+01))
      (:CastFailed
        (Utilities|Casting|CastToBP_Rufus_Water_LootLake Lake%20Reference
          (:then
            (Development|PrintString "Vehicle is entering Loot Lake.")
            (Variables|Default|SetWaterVFXOffset -1e+01))
          (:CastFailed
            (Development|PrintString "Vehicle is entering a Rufus Water Body.")
            (Variables|Default|SetWaterVFXOffset -6e+01)))))
    (else
      (Development|PrintString "Vehicle is exiting water."))))

(fn ConstructionScript
  (|CallParentFunction))
